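# Physics asset for the Noble gadget mesh, giving it a simple collision and physics shape.
# A single capsule body on the root bone stands in for the mesh, offset along the bone's Z axis.
# It has no constraints, so there is no articulation between parts and the mesh is represented as one body.
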PhysicsAsset /Game/Gadgets/Assets/Noble/Meshes/Noble_PhysicsAsset
bodies {root [{shapeType=Capsule, length=15.4029, radius=3.1291, center=<-0.0522 0.1035 12.1077>, rotation=<rot -0.2953 -90 -0.1073>}]}
constraints []
